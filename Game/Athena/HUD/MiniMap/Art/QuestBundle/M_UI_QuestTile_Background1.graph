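# Translucent UI material that draws a quest tile background with an offsettable, scalable item image.
# Composites a rotatable three-color gradient, the masked texture and a corner gradient, using rounded masks and viewport-scale-aware UVs.
# Scalar parameters such as IsHeader, IsTracked, QuestV3 and IsSideNav blend between layout and style variants via lerps rather than permutations.
# Also samples the Slate post buffer for a blurred backplate, and a static switch selects debug display.
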Material /Game/Athena/HUD/MiniMap/Art/QuestBundle/M_UI_QuestTile_Background1
graph {
 properties {bAutomaticallySetUsageInEditor=false, blendMode=BLEND_Translucent, materialDomain=MD_UI, usedShadingModels=Default Lit}
 nodes [
  n1 = MaterialFunctionCall!(Anchor, Translation, Padding, Scale) {materialFunction=/FortUILibrary/MaterialFunctions/Basics/MF_UI_VertexTransform}
  n2 = MaterialFunctionCall(Translation=Constant2Vector() {g=-12}, Padding=Constant2Vector() {g=24}) {materialFunction=&MF_UI_VertexTransform}
  LinearInterpolate(A=@n1.Transformed Position, B=@n2.Transformed Position, Alpha=NamedRerouteUsage())
  n3 = NamedRerouteUsage()
  n4 = NamedRerouteUsage()
  n5 = Multiply(
   A=MaterialFunctionCall(GradientStart=Constant() {r=0.35}, GradientEnd=Constant() {r=0.65}, RotationAngle=Constant() {r=-0.02}) {materialFunction=&MF_UI_LinearGradient}.Result
   B=LinearInterpolate(Alpha=MaterialFunctionCall() {materialFunction=&MF_UI_LinearGradient}.Result) {constA=1, constB=0.25}
  )
  n6 = ScalarParameter() {group=BP logic, parameterName=IsHeader}
  n7 = If(A=NamedRerouteUsage(), "A > B"=Constant(), "A == B"=@n6, "A < B"=@n6)
  Max(A=Multiply(A=Multiply(A=@n3, B=@n4), B=LinearInterpolate(B=@n5, Alpha=@n7) {constA=1, constAlpha=1}), B=NamedRerouteUsage())
  n8 = VectorParameter() {desc=LeftColor, group=1 - Background, parameterName=BackgroundColor, defaultValue=<rgba 0.007 0.1221 0.1329 1>}
  NamedRerouteDeclaration(None=@n8.RGB) {name=Background_Color, nodeColor=<rgba 1 0.9316 0.0156 1>}
  n9 = NamedRerouteUsage()
  n10 = NamedRerouteUsage()
  n11 = NamedRerouteUsage()
  n12 = ScalarParameter() {defaultValue=0.7, desc="GradientMidOffset\r\nGradientStart < GradientMid < GradientEnd", group=1 - Background, parameterName=2 GradientMid}
  n13 = ScalarParameter() {desc=From 0 to 1, group=1 - Background, parameterName=4 GradientRotation}
  n14 = ScalarParameter() {desc=Full gradient offset, group=1 - Background, parameterName=5 GradientOffset}
  n15 = MaterialFunctionCall(
   GradientStart=ScalarParameter() {desc="Gradient Start Offset \r\nGradientStart < GradientMid < GradientEnd", group=1 - Background, parameterName=1 GradientStart}
   GradientEnd=@n12
   RotationAngle=@n13
   GradientOffset=@n14
  ) {materialFunction=&MF_UI_LinearGradient}
  n16 = ScalarParameter() {defaultValue=1.6, desc="Gradient End Offset\r\nGradientStart < GradientMid < GradientEnd", group=1 - Background, parameterName=3 GradientEnd}
  n17 = MaterialFunctionCall(GradientStart=@n12, GradientEnd=@n16, RotationAngle=@n13, GradientOffset=@n14) {materialFunction=&MF_UI_LinearGradient}
  n18 = Multiply(A=Subtract(A=LinearInterpolate(Alpha=@n17.Result) {constA=0.5}) {constB=0.5}) {constB=2}
  n19 = Constant() {r=1}
  n20 = Constant() {r=1}
  n21 = Constant() {r=0.95}
  n22 = Constant() {r=1}
  n23 = LinearInterpolate(
   A=LinearInterpolate(
    B=MaterialFunctionCall!(
     Color A=@n9
     Color B=@n10
     Color C=@n11
     A to B Gradient=@n15.Result
     Color C Space=@n18
     "Display Color C (0 to 1)"=@n19
     "Saturation (-1 to 1)"=@n20
     Saturation Mask
     Saturation Mask Threshold=@n21
     "Color C Covers Saturation (0 to 1)"=@n22
    ) {materialFunction=&/MF_UI_Gradient_3Colors_WithSaturation}.Colorful RGBA Gradient
    Alpha=NamedRerouteUsage()
   )
   B=@n9
   Alpha=ScalarParameter() {group=BP logic, parameterName=IsHeader}
  )
  n24 = NamedRerouteUsage()
  n25 = NamedRerouteUsage()
  NamedRerouteDeclaration(None=LinearInterpolate(A=@n23, B=LinearInterpolate(A=@n24, B=@n23, Alpha=@n25), Alpha=NamedRerouteUsage())) {name=Background_RGB, nodeColor=<rgba 0 0.2462 1 1>}
  n26 = MaterialFunctionCall() {materialFunction=&GetUserInterfaceUV}
  n27 = LinearInterpolate(A=@n26.Normalized UV, B=Divide(A=OneMinus(None=@n26.Normalized UV), B=NamedRerouteUsage()), Alpha=NamedRerouteUsage()) {constA=1}
  n28 = LinearInterpolate(A=Constant() {r=16}, B=Constant() {r=12}, Alpha=NamedRerouteUsage())
  n29 = Constant() {r=1}
  n30 = Constant() {r=1}
  n31 = ScalarParameter() {defaultValue=0.25, group=SideNavigationStyle, parameterName=SideNav_CornerRadius}
  n32 = ScalarParameter() {group=SideNavigationStyle, parameterName=SideNav_RadiusInPixels}
  NamedRerouteDeclaration(None=LinearInterpolate(
    A=MaterialFunctionCall(UV=@n27, Size=@n30, Corner Radius=@n28, Radius in Pixels=@n29) {materialFunction=&MF_UI_CoreUI_RoundedMask}.Shape Mask
    B=MaterialFunctionCall(UV=@n27, Size=@n30, Corner Radius=@n31, Radius in Pixels=@n32) {materialFunction=&MF_UI_CoreUI_RoundedMask}.Shape Mask
    Alpha=ScalarParameter() {desc=Only for debugging. Should always to be 0, group=SideNavigationStyle, parameterName=IsSideNav}
   ) {constB=2}) {name=RoundedFill_Mask, nodeColor=<rgba 0 0.2462 1 1>}
  NamedRerouteDeclaration(None=LinearInterpolate(
    A=ScalarParameter() {defaultValue=1.05, group=2 - Texture, parameterName=Texture_Scale}
    B=ScalarParameter() {defaultValue=1.05, group=4 - Quest 3.0, parameterName=Quest3.0_Texture_Scale}
    Alpha=NamedRerouteUsage()
   )) {name=Texture_Scale, nodeColor=<rgba 0.0262 1 0 1>}
  n33 = LinearInterpolate(
   A=ScalarParameter() {defaultValue=-50, group=2 - Texture, parameterName=Texture_XOffset, sliderMax=100, sliderMin=-100}
   B=ScalarParameter() {defaultValue=-10, group=4 - Quest 3.0, parameterName=Quest3.0_Texture_XOffset, sliderMax=100, sliderMin=-100}
   Alpha=NamedRerouteUsage()
  )
  n34 = ComponentMask(None=NamedRerouteUsage()) {r=true}
  NamedRerouteDeclaration(None=Divide(A=Multiply(A=@n33) {constB=-1}, B=Max(A=@n34))) {name=Texture_XOffset, nodeColor=<rgba 0.0257 1 0 1>}
  n35 = MaterialFunctionCall() {materialFunction=&GetUserInterfaceUV}
  NamedRerouteDeclaration(None=@n35.Pixel Size) {name=PixelSize, nodeColor=<rgba 0.0262 1 0 1>}
  n36 = LinearInterpolate(
   A=ScalarParameter() {group=2 - Texture, parameterName=Texture_YOffset}
   B=ScalarParameter() {group=4 - Quest 3.0, parameterName=Quest3.0_Texture_YOffset}
   Alpha=NamedRerouteUsage()
  )
  n37 = ComponentMask(None=NamedRerouteUsage()) {g=true}
  NamedRerouteDeclaration(None=Divide(A=Multiply(A=@n36) {constB=-1}, B=Max(A=@n37))) {name=Texture_YOffset, nodeColor=<rgba 0.0257 1 0 1>}
  n38 = TextureObjectParameter() {group=2 - Texture, parameterName=Texture, texture=/Game/2dAssets/Emoji/Season4/Crab}
  NamedRerouteDeclaration(None=TextureProperty(Tex=@n38)) {name=Texture_Size, nodeColor=<rgba 0.0257 1 0 1>}
  NamedRerouteDeclaration(None=@n38) {name=Texture_Object, nodeColor=<rgba 0 0.2461 1 1>}
  n39 = NamedRerouteUsage()
  n40 = Add(A=Multiply(A=OneMinus(None=@n39)) {constB=1.5})
  n41 = AppendVector(A=ComponentMask(None=@n40) {g=true}, B=ComponentMask(None=@n40) {r=true})
  n42 = NamedRerouteUsage()
  n43 = Constant() {desc="Please don't use this\r\nBetter use Texture_XPixelOffset", r=0.5}
  n44 = Constant() {desc="Please don't use this\r\nBetter use Texture_YPixelOffset", r=0.5}
  n45 = NamedRerouteUsage()
  n46 = Divide(
   A=ComponentMask(None=NamedRerouteUsage()) {g=true}
   B=Multiply(
    A=Constant() {r=134}
    B=Custom() {code="return View.ViewportScaleUI;", description=ViewportScaleFactor, outputType=CMOT_Float1, showCode=true, inputs []}
   )
  ) {constA=144, constB=114}
  n47 = ScalarParameter() {defaultValue=1, parameterName=UseAlternativeHeight}
  n48 = Add(A=ScalarParameter() {group=BP logic, parameterName=IsHeader}, B=If(A=NamedRerouteUsage(), "A > B"=Constant(), "A == B"=@n47, "A < B"=@n47))
  n49 = AppendVector(
   A=LinearInterpolate(A=NamedRerouteUsage(), Alpha=ScalarParameter() {parameterName=IsNotification}) {constB=0}
   B=LinearInterpolate(A=@n45, B=Multiply(A=@n46, B=@n45), Alpha=Saturate(None=@n48))
  )
  n50 = Custom() {code="return View.ViewportScaleUI;", description=ViewportScaleFactor, outputType=CMOT_Float1, showCode=true, inputs []}
  n51 = MaterialFunctionCall!(
   Texture Scale=LinearInterpolate(A=@n42, B=Multiply(A=@n41, B=@n42), Alpha=NamedRerouteUsage())
   UVs=Add(A=NamedRerouteUsage(), B=Multiply(A=@n49, B=@n50))
   Offset XY Texturesize Z=AppendVector(A=AppendVector(A=@n43, B=@n44), B=Constant() {r=0.5})
  ) {materialFunction=/Engine/Functions/Engine_MaterialFunctions01/Texturing/ScaleUVsAroundPoint}
  n52 = NamedRerouteUsage()
  n53 = Constant() {r=1}
  n54 = TextureSample(
   UVs=NamedRerouteDeclaration(None=MaterialFunctionCall(UV=@n51.UVs, TextureSize=@n52, ScalePivotPoint=@n53) {materialFunction=&MF_UI_UVScaling}.ScaleY) {name=TextureUVs, nodeColor=<rgba 0 0.2462 1 1>}
   Tex=NamedRerouteUsage()
  ) {samplerSource=SSM_Clamp_WorldGroupSettings}
  n55 = Constant() {r=1}
  n56 = MaterialFunctionCall!(UV=NamedRerouteUsage(), Grid Size=@n55, Index=@n55) {materialFunction=/FortUILibrary/MaterialFunctions/Basics/MF_UI_IsolateUV}
  NamedRerouteDeclaration(None=Multiply(A=@n54.RGB, B=@n56.Result)) {name=Texture_RGB, nodeColor=<rgba 0 0.2461 1 1>}
  n57 = ScalarParameter() {defaultValue=1, desc="1 = texture fully opaque\r\n0 = texture completely transparent", group=2 - Texture, parameterName=ShowItemImage}
  n58 = Multiply(A=LinearInterpolate(A=@n57, B=Ceil(None=@n57), Alpha=NamedRerouteUsage()), B=Multiply(A=@n54.A, B=@n56.Result))
  n59 = Multiply(A=@n58, B=NamedRerouteUsage())
  n60 = NamedRerouteUsage()
  n61 = MaterialFunctionCall(GradientStart=Constant(), GradientEnd=Constant() {r=0.2}, RotationAngle=Constant() {r=0.4}, GradientOffset=Constant() {r=-0.2}) {materialFunction=&MF_UI_LinearGradient}
  NamedRerouteDeclaration(None=LinearInterpolate(
    A=LinearInterpolate(A=@n59, B=@n58, Alpha=@n60)
    B=Multiply(A=@n58, B=@n61.Result)
    Alpha=ScalarParameter() {group=BP logic, parameterName=IsHeader}
   )) {name=Texture_Alpha, nodeColor=<rgba 0 0.2462 1 1>}
  n62 = ScalarParameter() {
   defaultValue=1.9
   desc="Gradient softness\r\n0 = hard, sharp edge \r\n1 = smooth and soft"
   group=CornerGradient
   parameterName=CornerGradient_SmoothStepMax
  }
  NamedRerouteDeclaration(None=@n62) {name=CornerGradient_SmoothStepMax, nodeColor=<rgba 0.0257 1 0 1>}
  n63 = ScalarParameter() {defaultValue=3, desc="Don't use this please\r\nBetter use CornerGradient_Size", group=Legacy, parameterName=CornerGradient_XPosition}
  n64 = ScalarParameter() {desc="Don't use this please\r\nBetter use CornerGradient_Size", group=Legacy, parameterName=CornerGradient_YPosition}
  n65 = ScalarParameter() {defaultValue=4, group=CornerGradient, parameterName=CornerGradient_Size}
  n66 = MaterialFunctionCall(
   X=Constant() {desc="Don't use this please\r\nBetter use CornerGradient_Size", r=3}
   Y=Constant() {desc="Don't use this please\r\nBetter use CornerGradient_Size", r=0.4}
   Z=Constant() {r=3}
  ) {materialFunction=&MakeFloat3}
  n67 = ScalarParameter() {parameterName=IsTracked}
  NamedRerouteDeclaration(None=LinearInterpolate(
    A=MaterialFunctionCall(X=@n63, Y=@n64, Z=@n65) {materialFunction=&MakeFloat3}.Result
    B=LinearInterpolate(A=@n66.Result, Alpha=@n67) {constB=0}
    Alpha=NamedRerouteUsage()
   ) {constB=0}) {name=CornerGradient_AnchorPosition, nodeColor=<rgba 0 0.2462 1 1>}
  n68 = NamedRerouteUsage()
  n69 = Constant() {r=0.3}
  n70 = AppendVector(A=@n69, B=@n69)
  n71 = NamedRerouteUsage()
  n72 = MaterialFunctionCall(UV Channel=MaterialFunctionCall(
    UVs=MaterialFunctionCall(Texture Scale=@n70, Offset XY Texturesize Z=@n71) {materialFunction=&ScaleUVsAroundPoint}.UVs
    Rotation Center=Constant() {r=3}
    "Rotation Angle (0-1)"=Constant() {r=-0.1}
   ) {materialFunction=&CustomRotator}.Rotated Values) {materialFunction=&LinearGradient}
  NamedRerouteDeclaration(None=Multiply(A=SmoothStep(Max=@n68, Value=@n72.UGradient) {constMax=1.5}, B=NamedRerouteUsage())) {name=CornerGradient_Mask, nodeColor=<rgba 0 0.2461 1 1>}
  n73 = VectorParameter() {group=CornerGradient, parameterName=CornerGradient_ColorFade, defaultValue=<rgba 0.0482 0.1329 0.0194 1>}
  NamedRerouteDeclaration(None=@n73.RGB) {name=CornerGradient_RGB, nodeColor=<rgba 0 0.2461 1 1>}
  n74 = ScalarParameter() {group=BP logic, parameterName=IsHeader}
  NamedRerouteDeclaration(None=Multiply(A=ComponentMask(None=@n73.A) {a=true}, B=OneMinus(None=@n74))) {name=CornerGradient_Alpha, nodeColor=<rgba 0 0.2461 1 1>}
  n75 = VectorParameter() {group=Debug, parameterName=BackgroundDefaultColor, defaultValue=<rgba 0.099 0.099 0.099 0.5>}
  NamedRerouteDeclaration(None=@n75.RGB) {name=Background_DebugColor, nodeColor=<rgba 1 0.9316 0.0156 1>}
  NamedRerouteDeclaration(None=@n75.A) {name=Background_DebugAlpha, nodeColor=<rgba 1 0.9316 0.0156 1>}
  n76 = MaterialFunctionCall!("UVs (Optional)") {materialFunction=/Engine/Functions/UserInterface/GetSlatePost0}
  n77 = MaterialFunctionCall!(Color=Max(A=Saturate(None=@n76.LinearRGB)) {constB=0.0001}, Contrast=ScalarParameter() {defaultValue=0.4, parameterName=EffectWeakness}) {materialFunction=/Engine/Functions/Engine_MaterialFunctions01/ImageAdjustment/Contrast_Preserve_Color}
  NamedRerouteDeclaration(None=MaterialFunctionCall!(Blend=Saturate(None=@n77.Result), Base=NamedRerouteUsage()) {materialFunction=/Engine/Functions/Engine_MaterialFunctions03/Blends/Blend_Screen}.Result) {name=Background_DebugRGB, nodeColor=<rgba 0 0.2461 1 1>}
  n78 = NamedRerouteUsage()
  n79 = NamedRerouteUsage()
  n80 = ScalarParameter() {defaultValue=3, group=SideNavigationStyle, parameterName=SN_Scale}
  n81 = Multiply(A=@n80, B=Divide(A=ComponentMask(None=@n79) {r=true}, B=ComponentMask(None=@n79) {g=true}))
  n82 = Divide(A=AppendVector(
    A=ScalarParameter() {defaultValue=57, group=SideNavigationStyle, parameterName=SN_OffsetX}
    B=ScalarParameter() {defaultValue=13, group=SideNavigationStyle, parameterName=SN_OffsetY}
   )) {constB=100}
  n83 = Constant() {r=1}
  n84 = TextureSample(
   UVs=MaterialFunctionCall(Texture Scale=AppendVector(A=@n81, B=@n80), UVs=NamedRerouteUsage(), Offset XY Texturesize Z=AppendVector(A=@n82, B=@n83)) {materialFunction=&ScaleUVsAroundPoint}.UVs
   Tex=NamedRerouteUsage()
  ) {samplerSource=SSM_Clamp_WorldGroupSettings}
  n85 = ScalarParameter() {desc=Only for debugging. Should always to be 0, group=SideNavigationStyle, parameterName=IsSideNav}
  n86 = NamedRerouteUsage()
  n87 = ScalarParameter() {defaultValue=1, desc="1 = texture fully opaque\r\n0 = texture completely transparent", group=2 - Texture, parameterName=ShowItemImage}
  n88 = Ceil(None=@n87)
  n89 = NamedRerouteUsage()
  n90 = LinearInterpolate(A=NamedRerouteUsage(), B=Multiply(A=Multiply(A=@n84.A, B=@n86), B=LinearInterpolate(A=@n87, B=@n88, Alpha=@n89)), Alpha=@n85) {constB=2}
  n91 = LinearInterpolate(
   A=LinearInterpolate(A=NamedRerouteUsage(), B=LinearInterpolate(A=@n78, B=@n84.RGB, Alpha=@n85) {constB=2}, Alpha=@n90) {constAlpha=0}
   B=NamedRerouteUsage()
   Alpha=LinearInterpolate(A=NamedRerouteUsage(), Alpha=@n85) {constB=0}
  )
  n92 = MaterialFunctionCall(GradientStart=Constant() {r=0.4}, GradientEnd=Constant() {r=1.3}, RotationAngle=Constant() {r=0.75}) {materialFunction=&MF_UI_LinearGradient}
  n93 = LinearInterpolate(B=Constant3Vector() {constant=<rgba 0.0048 0.004 0.0116 1>}, Alpha=Saturate(None=SmoothStep(Value=@n92.Result) {constMin=0.1, constValue=1})) {constA=1}
  n94 = NamedRerouteUsage()
  n95 = ScalarParameter() {group=BP logic, parameterName=IsHeader}
  NamedRerouteDeclaration(None=LinearInterpolate(A=Multiply(A=@n91, B=@n93), B=@n91, Alpha=Max(A=@n94, B=@n95))) {name=CompositeBackground_RGB, nodeColor=<rgba 0 0.2462 1 1>}
  NamedRerouteDeclaration(None=@n35.9-Slice UV) {name=UVs, nodeColor=<rgba 0.0262 1 0 1>}
  n96 = VectorParameter() {desc=CenterColor, group=1 - Background, parameterName=ColorSecond, defaultValue=<rgba 0.1413 0.5149 0.0065 1>}
  NamedRerouteDeclaration(None=@n96.RGB) {name=Background_ColorSecond, nodeColor=<rgba 1 0.9316 0.0156 1>}
  n97 = VectorParameter() {desc=RightColor, group=1 - Background, parameterName=ColorCenter, defaultValue=<rgba 0.3095 1 0.0137 1>}
  NamedRerouteDeclaration(None=@n97.RGB) {name=Background_ColorCenter, nodeColor=<rgba 1 0.9316 0.0156 1>}
  n98 = ScalarParameter() {
   defaultValue=0.7
   desc="0 = completely transparent \r\n1 = fully opaque "
   group=1 - Background
   parameterName=Background Opacity
   sliderMax=1
  }
  n99 = LinearInterpolate(A=@n98, Alpha=NamedRerouteUsage())
  n100 = ScalarParameter() {group=BP logic, parameterName=IsHeader}
  NamedRerouteDeclaration(None=StaticSwitchParameter(True=NamedRerouteUsage(), False=LinearInterpolate(A=@n99, Alpha=@n100)) {parameterName=IsDebugDisplay}) {name=Background_Opacity, nodeColor=<rgba 0 0.2461 1 1>}
  NamedRerouteDeclaration(None=@n90) {name=Alpha, nodeColor=<rgba 0 0.1328 1 1>}
  NamedRerouteDeclaration(None=Divide(A=@n1.New Pixel Size, B=@n2.New Pixel Size)) {name=QuestV3NewPixelSize, nodeColor=<rgba 0.0547 0 1 1>}
  n101 = ScalarParameter() {defaultValue=0.4, group=4 - Quest 3.0, parameterName=Quest3.0_GradientStart}
  n102 = Constant() {r=1}
  n103 = Constant()
  NamedRerouteDeclaration(None=LinearInterpolate(
    A=MaterialFunctionCall(GradientStart=@n101, GradientEnd=@n102, RotationAngle=@n103) {materialFunction=&MF_UI_LinearGradient}.Result
    Alpha=ScalarParameter() {parameterName=IsTracked}
   ) {constB=0}) {name=Gradient_Quest3.0, nodeColor=<rgba 0 1 0.2266 1>}
  NamedRerouteDeclaration(None=MaterialFunctionCall() {materialFunction=&GetSlatePost0}.LinearRGB) {name=Blur RGB, nodeColor=<rgba 0 0 0 1>}
  NamedRerouteDeclaration(None=MaterialFunctionCall!(Post Buffer Blur=NamedRerouteUsage()) {materialFunction=/FortUILibrary/MaterialFunctions/CoreUI/MF_UI_CoreUI_BackplateB}.Weak) {name=BackplateB_Weak, nodeColor=<rgba 0 0.2462 1 1>}
  NamedRerouteDeclaration(None=ScalarParameter() {group=4 - Quest 3.0, parameterName=QuestV3}) {name=QuestV3, nodeColor=<rgba 0.0547 1 0 1>}
  NamedRerouteDeclaration(None=@n2.Total Scale Factor) {name=NewPixelSixe, nodeColor=<rgba 0 1 0.0156 1>}
  StaticSwitchParameter(True=NamedRerouteUsage(), False=NamedRerouteUsage()) {parameterName=IsDebugDisplay}
 ]
}
