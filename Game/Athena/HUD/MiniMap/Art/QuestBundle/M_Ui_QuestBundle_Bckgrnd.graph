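# Translucent UI material that draws the quest bundle background in the HUD, with an animated triangle pattern and states for selection, focus and burst.
# A triangle grid, rotated and panned in aspect-corrected screen space and masked by a horizontal gradient, blends between color pairs.
# Burst moves an overlay brightening edge, while Select and Focus switch palettes.
# A UI box function shapes the panel, rounding right-side corners as Select rises.
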Material /Game/Athena/HUD/MiniMap/Art/QuestBundle/M_Ui_QuestBundle_Bckgrnd
graph {
 properties {blendMode=BLEND_Translucent, materialDomain=MD_UI, pixelDepthOffsetMode=PDOM_Legacy, usedShadingModels=Default Lit}
 nodes [
  n1 = MaterialFunctionCall(Float2=MaterialFunctionCall() {materialFunction=&GetUserInterfaceUV}.Pixel Size) {materialFunction=&BreakOutFloat2Components}
  AppendVector(A=Divide(A=@n1.R, B=@n1.R), B=Constant() {r=1})
  Constant() {r=1}
  n2 = Constant() {r=1}
  n3 = MaterialFunctionCall(Float2=MaterialFunctionCall() {materialFunction=&ScreenResolution}.Visible Resolution) {materialFunction=&BreakOutFloat2Components}
  n4 = Divide(A=@n3.R, B=Max(A=@n3.G))
  n5 = Constant() {r=1}
  n6 = Rotator(
   Coordinate=Multiply(
    A=AppendVector(A=@n4, B=@n5)
    B=MaterialFunctionCall() {materialFunction=&ScreenAlignedUVs}."X 100%, Y 100%"
   )
   Time=ScalarParameter() {defaultValue=-0.0571, group=Triangles, parameterName=Tri Angle}
  ) {speed=6.2832}
  n7 = AppendVector(A=ScalarParameter() {defaultValue=0.0025, group=Triangles, parameterName=PanSpeed}, B=Constant())
  n8 = TextureSample(UVs=Multiply(A=Panner(Coordinate=@n6, Speed=@n7), B=ScalarParameter() {defaultValue=5, group=Triangles, parameterName=Tri Tile Amount})) {texture=/Game/Global/Textures/Default/Masks_and_Gradients/T-TriangleGrid}
  n9 = LinearInterpolate(
   A=ScalarParameter() {defaultValue=0.2, group=Triangles, parameterName=TriangleDensity}
   B=ScalarParameter() {defaultValue=0.2, group=Triangles, parameterName=TriangleDensity_Select}
   Alpha=ScalarParameter() {parameterName=Select, sliderMax=1}
  )
  n10 = ComponentMask(None=Subtract(
    A=MaterialFunctionCall() {materialFunction=&GetUserInterfaceUV}.Normalized UV
    B=ScalarParameter() {group=Triangles, parameterName=TriangleMask_Position}
   )) {r=true}
  n11 = ScalarParameter() {defaultValue=1.5, group=Triangles, parameterName=TriangleMask_sharpness}
  n12 = Multiply(A=Saturate(None=Multiply(A=@n8.R, B=@n9)), B=Saturate(None=Multiply(A=@n10, B=@n11)))
  n13 = LinearInterpolate(
   A=VectorParameter() {group=colors, parameterName=Base_RGB1, defaultValue=<rgba 0.0003 0.0185 0.2747 1>}.RGB
   B=VectorParameter() {group=colors, parameterName=Base_RGB2, defaultValue=<rgba 0 0.2242 1 1>}.RGB
   Alpha=@n12
  )
  n14 = MaterialFunctionCall() {materialFunction=&GetUserInterfaceUV}
  n15 = LinearInterpolate(
   A=ScalarParameter() {group=Triangles, parameterName=Busrt_Position_Start}
   B=ScalarParameter() {group=Triangles, parameterName=Busrt_Position_End}
   Alpha=ScalarParameter() {parameterName=Burst, sliderMax=1}
  )
  n16 = Multiply(
   A=ComponentMask(None=Subtract(A=@n14.Normalized UV, B=@n15)) {r=true}
   B=ScalarParameter() {defaultValue=1.5, group=Triangles, parameterName=Burst_sharpness}
  )
  n17 = VectorParameter() {group=colors, parameterName=OffFocus_RGB1, defaultValue=<rgba 0 0.2242 1 1>}
  n18 = VectorParameter() {group=colors, parameterName=OffFocus_RGB2, defaultValue=<rgba 0.0781 0.8343 1 1>}
  n19 = VectorParameter() {group=colors, parameterName=Selected_RGB1, defaultValue=<rgba 1 1 1 1>}
  n20 = VectorParameter() {group=colors, parameterName=Selected_RGB2, defaultValue=<rgba 0.313 0.3397 0.4323 1>}
  LinearInterpolate(
   A=LinearInterpolate(
    A=@n13
    B=MaterialFunctionCall(Base=@n13, Blend=@n2) {materialFunction=&Blend_Overlay}.Result
    Alpha=Saturate(None=@n16)
   )
   B=LinearInterpolate(
    A=LinearInterpolate(A=@n17.RGB, B=@n18.RGB, Alpha=@n12)
    B=LinearInterpolate(A=@n19.RGB, B=@n20.RGB, Alpha=@n12)
    Alpha=ScalarParameter() {defaultValue=1, parameterName=Focus, sliderMax=1}
   )
   Alpha=ScalarParameter() {parameterName=Select, sliderMax=1}
  )
  n21 = Round(None=ComponentMask(None=TextureCoordinate()) {r=true})
  n22 = ScalarParameter() {parameterName=Select, sliderMax=1}
  MaterialFunctionCall(
   Size=Constant() {r=1}
   CornerRoundness=LinearInterpolate(B=ScalarParameter() {defaultValue=0.2, parameterName=Corner Roundness}, Alpha=Multiply(A=@n21, B=@n22))
   KeepAspectRatio=StaticBool() {value=true}
  ) {materialFunction=&MF_UI_Box}
 ]
}
